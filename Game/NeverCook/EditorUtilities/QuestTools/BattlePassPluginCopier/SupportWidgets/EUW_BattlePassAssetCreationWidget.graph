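# Editor utility widget that creates a new battle pass season's umbrella asset and XP tables from the previous season's.
# Season changes rerun asset registry and file checks that set readiness flags gating the buttons.
# Umbrella creation duplicates the asset and rewrites its FirstWin season tag; XP creation copies the spreadsheet, duplicates the DataTable and exports CSV.
# The Rest Tables button and MakeStoreAssets have no working logic.
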EditorUtilityWidgetBlueprint /Game/NeverCook/EditorUtilities/QuestTools/BattlePassPluginCopier/SupportWidgets/EUW_BattlePassAssetCreationWidget
graphs [
 EventGraph {nodes [
   "On Clicked (TA_QuestEditor_Button)"!() -> (OutputDelegate, then, widget:User Widget Object Reference, event:Pointer Event Structure) {is=title K2Node_ComponentBoundEvent}
   > Branch(Condition=Get UmbrellaReadyToCopy() -> (UmbrellaReadyToCopy:bool))
   > MakeUmbrella()
   "On Clicked (MakeXPSpreadsheetButton)"!() -> (OutputDelegate, then, widget:User Widget Object Reference, event:Pointer Event Structure) {is=title K2Node_ComponentBoundEvent}
   > Branch(Condition=Get XpTableIsReady() -> (XpTableIsReady:bool))
   > MakeXpCurves()
   "On Property Changed (DeleteAfterTesting)"!() -> (OutputDelegate, then, PropertyName:Name=None) {is=title K2Node_ComponentBoundEvent}
   > ReadySeasonUmbrella()
   > ReadyXPDataTable()
   Event PreConstruct()
   > SetObject(self=Get DeleteAfterTesting() -> (DeleteAfterTesting:Details View Object Reference), NewObject=Self-Reference())
  ]}
 MakeStoreAssets {nodes [
   MakeStoreAssets() {is=title K2Node_FunctionEntry}
   Get SeasonItemDefintion() -> (SeasonItemDefintion:Athena Season Item Definition Object Reference)
  ]}
 MakeUmbrella {nodes [
   MakeUmbrella() {is=title K2Node_FunctionEntry}
   > n1 = GetAssetTools()
   n2 = Get PrevSeasonUmbrella() -> (PrevSeasonUmbrella:Asset Data Structure)
   n3 = Get SeasonNumber() -> (SeasonNumber:int=0)
   > n4 = DuplicateAsset!(
    execute
    self:Asset Tools Interface=@n1
    AssetName:String=Append(A=Umbrella_Season_, B="To String (Integer)"(InInt=@n3))
    PackagePath:String="To String (Name)"(InName=Break Asset Data!(AssetData=@n2) -> (PackagePath, AssetClassPath))
    OriginalObject:Object Reference=GetAsset(InAssetData=@n2)
   ) -> (then, ReturnValue:Object Reference)
   > n5 = Set NewUmbrella(NewUmbrella:Object Reference=@n4) -> (Output_Get:Object Reference)
   > n6 = GetEditorProperty!(execute:Exec, Object:Object Reference=@n5, PropertyName:Name=Get GameplayTagContainerName() -> (GameplayTagContainerName:Name=None)) -> (then:Exec, PropertyValue:Gameplay Tag Container Structure, ReturnValue:bool)
   > Set gameplayTagsToWrite(gameplayTagsToWrite:Gameplay Tag Container Structure=@n6) -> (Output_Get:Gameplay Tag Container Structure)
   n7 = int - int(A=Get SeasonNumber() -> (SeasonNumber=0))
   > n8 = RequestGameplayTag!(
    execute
    TagName:String=Append!(A=Cosmetics.Source.Season, B="To String (Integer)"(InInt=@n7), C:String=.FirstWin) -> (ReturnValue)
    bErrorIfNotFound:bool=false
   ) -> (then, OutTag:Gameplay Tag Structure, ReturnValue:bool)
   > RemoveGameplayTag!(
    execute
    TagContainer:"Gameplay Tag Container Structure (by ref)"=Get gameplayTagsToWrite() -> (gameplayTagsToWrite:Gameplay Tag Container Structure)
    Tag:Gameplay Tag Structure=@n8
   ) -> (then, ReturnValue:bool)
   n9 = Get SeasonNumber() -> (SeasonNumber=0)
   > n10 = RequestGameplayTag(TagName=Append(A=Cosmetics.Source.Season, B="To String (Integer)"(InInt=@n9), C=.FirstWin), bErrorIfNotFound=false)
   > AddGameplayTag!(execute, TagContainer:"Gameplay Tag Container Structure (by ref)"=Get gameplayTagsToWrite(), Tag:Gameplay Tag Structure=@n10) -> (then)
   > SetEditorProperty(
    Object=Get NewUmbrella() -> (NewUmbrella:Object Reference)
    PropertyName=Get GameplayTagContainerName() -> (GameplayTagContainerName=None)
    PropertyValue:Gameplay Tag Container Structure=Get gameplayTagsToWrite()
   )
   > SaveLoadedAsset(AssetToSave=Get NewUmbrella(), bOnlyIfIsDirty=true)
   > "SetText (Text)"(self=Get UmbrellaOutput() -> (UmbrellaOutput:Common Text Object Reference), InText="LOCTEXT(\"Umbrella Created!\")")
   SetEditorProperty(Object=Get SeasonItemDefintion(), PropertyName=SeasonFirstWinRewards, PropertyValue:"Wildcard (by ref)")
   Get NewUmbrella()
  ]}
 MakeXpCurves {nodes [
   MakeXpCurves() {is=title K2Node_FunctionEntry}
   n1 = Get XpSpreadsheetPrevSeasonPath() -> (XpSpreadsheetPrevSeasonPath:String)
   n2 = Get XpSpreadsheetNewSeasonPath() -> (XpSpreadsheetNewSeasonPath:String)
   > PrintString(InString=Append!(A="Copying XP Spreadsheet from: ", B=@n1, C=" to: ", D:String=@n2) -> (ReturnValue))
   > CopyFile!(execute, InSourceFile:String=@n1, InDestFile:String=@n2, bSilent:bool=false) -> (then, ReturnValue:bool)
   > n3 = GetAssetTools()
   n4 = Get XpDataTablePrevSeason() -> (XpDataTablePrevSeason:Asset Data Structure)
   n5 = Get SeasonNumber() -> (SeasonNumber=0)
   n6 = Replace(
    SourceString="To String (Name)"(InName=Break Asset Data!(AssetData=@n4) -> (PackageName:Name, PackagePath, AssetName:Name, AssetClassPath).PackagePath)
    From="To String (Integer)"(InInt=int - int(A=@n5))
    To="To String (Integer)"(InInt=Get SeasonNumber() -> (SeasonNumber=0))
   )
   n7 = GetAsset(InAssetData=@n4)
   n8 = Append!(A=Athena_Season, B="To String (Integer)"(InInt=Get SeasonNumber() -> (SeasonNumber=0)), C=_XP) -> (ReturnValue)
   > n9 = DuplicateAsset(self=@n3, AssetName=@n8, PackagePath=@n6, OriginalObject=@n7)
   > PrintString(InString=Append!(A="Duplicated XP Data Table from: ", B=Get Object Path String(Object=@n7), C=" to: ", D:String=@n6, E:String=@n8) -> (ReturnValue))
   > n10 = Cast To DataTable!(execute, Object=@n9) -> (then, CastFailed, AsData Table:Data Table Object Reference) {is=title K2Node_DynamicCast}
   n11 = Break Asset Data(AssetData=CreateAssetData(InAsset=@n10, bAllowBlueprintClass=false))
   n12 = Get SeasonNumber() -> (SeasonNumber=0)
   n13 = int - int(A=Get SeasonNumber() -> (SeasonNumber=0))
   > n14 = ConvertAssetPathToFilePath!(
    execute
    InAssetPath:String=Replace(SourceString="To String (Name)"(InName=@n11), From="To String (Integer)"(InInt=@n13), To="To String (Integer)"(InInt=@n12))
    InExtension:String=.csv
   ) -> (then, OutPath:String, ReturnValue:bool)
   > PrintString(InString=Append!(A="Exporting Data Table to: ", B=@n14) -> (ReturnValue))
   > ExportDataTableToCSV!(execute, DataTable:Data Table Object Reference=@n10, bShowFilePickerDialog:bool=false, PathToWriteDataTable:String=@n14) -> (then, ReturnValue:bool)
  ]}
 ReadySeasonUmbrella {nodes [
   ReadySeasonUmbrella() {is=title K2Node_FunctionEntry}
   > Set UmbrellaReadyToCopy(UmbrellaReadyToCopy:bool=false) -> (Output_Get:bool)
   n1 = "To String (Integer)"(InInt=Get SeasonNumber() -> (SeasonNumber=0))
   > n2 = GetAssetsByPackageName!(
    execute
    self:Asset Registry Interface=GetAssetRegistry()
    PackageName:Name=String To Name(InString=Append(A=Umbrella_Season_, B=@n1))
    bIncludeOnlyOnDiskAssets:bool=false
    bSkipARFilteredAssets:bool=true
   ) -> (then, OutAssetData:Array of Asset Data Structures, ReturnValue:bool)
   > n3 = Branch(Condition=@n2.ReturnValue)
   > "SetText (Text)"(self=Get PrevUmbrellaText() -> (PrevUmbrellaText:Common Text Object Reference), InText="LOCTEXT(\"Umbrella Already Exists.\")")
   > setLabel!(
    execute
    self:TA Base Editor Button Widget Object Reference=Get TA_QuestEditor_Button() -> (TA_QuestEditor_Button:TA Quest Editor Button Object Reference)
    label:String=Nothing To Do
   ) -> (then)
   n4 = "To String (Integer)"(InInt=int - int(A=Get SeasonNumber() -> (SeasonNumber=0)))
   n5 = GetAssetsByPackageName(
    execute=@n3.else
    self=GetAssetRegistry()
    PackageName=String To Name(InString=Append(A=Umbrella_Season_, B=@n4))
    bIncludeOnlyOnDiskAssets=false
    bSkipARFilteredAssets=true
   )
   > n6 = Branch(Condition=@n5.ReturnValue)
   > Set PrevSeasonUmbrella(PrevSeasonUmbrella:Asset Data Structure="Get (a ref)"(Array:Array of Asset Data Structures=@n5, Dimension 1=0) -> (Output:"Asset Data Structure (by ref)")) -> (Output_Get:Asset Data Structure)
   > "SetText (Text)"(self=Get PrevUmbrellaText(), InText="LOCTEXT(\"Found Previous Season Umbrella.\")")
   n7 = Get SeasonNumber() -> (SeasonNumber=0)
   > setLabel(
    self=Get TA_QuestEditor_Button()
    label=Append!(A="Create Season ", B="To String (Integer)"(InInt=@n7), C:String=" umbrella") -> (ReturnValue)
   )
   > Set UmbrellaReadyToCopy(UmbrellaReadyToCopy=true)
   "SetText (Text)"(
    execute=@n6.else
    self=Get PrevUmbrellaText()
    InText="LOCTEXT(\"Couldn\\'t find the previous season\\'s umbrella. Can\\'t make a new one automatically.\")"
   )
   > setLabel(self=Get TA_QuestEditor_Button(), label=Nothing To Do)
  ]}
 ReadyXPDataTable {nodes [
   ReadyXPDataTable() {is=title K2Node_FunctionEntry}
   > n1 = Sequence!(execute) -> (then_0, then_1, then_2:Exec, then_3:Exec)
   > Set AllStepsSuccessful(AllStepsSuccessful:bool=true) -> (Output_Get:bool)
   n2 = Sequence!(execute=@n1.then_1) -> (then_0, then_1, then_2, then_3, then_4:Exec, then_5:Exec)
   n3 = "To String (Integer)"(InInt=int - int(A=Get SeasonNumber() -> (SeasonNumber=0)))
   n4 = GetAssetsByPackageName(
    execute=@n2.then_1
    self=GetAssetRegistry()
    PackageName=String To Name(InString=Append(A=Athena_Season, B=@n3, C=_Extended))
    bIncludeOnlyOnDiskAssets=false
    bSkipARFilteredAssets=true
   )
   > n5 = Branch(Condition=@n4.ReturnValue)
   > Set XpDataTablePrevSeasonExtended(XpDataTablePrevSeasonExtended:Asset Data Structure="Get (a ref)"(Array=@n4, Dimension 1=0)) -> (Output_Get:Asset Data Structure)
   Set AllStepsSuccessful(execute=@n5.else, AllStepsSuccessful=false)
   n6 = Get SeasonNumber() -> (SeasonNumber=0)
   n7 = Append(A=Athena_Season, B="To String (Integer)"(InInt=int - int(A=@n6)), C=_XP)
   PrintString(execute=@n2.then_0, InString=Append!(A="Finding Season XP Asset: ", B=@n7) -> (ReturnValue))
   > n8 = GetAssetsByPackageName(self=GetAssetRegistry(), PackageName=String To Name(InString=@n7), bIncludeOnlyOnDiskAssets=false, bSkipARFilteredAssets=true)
   > n9 = Branch(Condition=@n8.ReturnValue)
   > Set XpDataTablePrevSeason(XpDataTablePrevSeason:Asset Data Structure="Get (a ref)"(Array=@n8, Dimension 1=0)) -> (Output_Get:Asset Data Structure)
   Set AllStepsSuccessful(execute=@n9.else, AllStepsSuccessful=false)
   Set DataTablesPath(execute=@n2.then_2, DataTablesPath:String=Append(A=GetProjectDirectory!() -> (ReturnValue:String), B=DataTables/)) -> (Output_Get:String)
   n10 = "To String (Integer)"(InInt=int - int(A=Get SeasonNumber() -> (SeasonNumber=0)))
   Set XpSpreadsheetPrevSeasonPath(
    execute=@n2.then_3
    XpSpreadsheetPrevSeasonPath:String=Append(A=Get DataTablesPath() -> (DataTablesPath:String), B=Append!(A=AthenaXPLevels_S, B=@n10, C:String=.xlsb) -> (ReturnValue))
   ) -> (Output_Get:String)
   n11 = "To String (Integer)"(InInt=Get SeasonNumber() -> (SeasonNumber=0))
   Set XpSpreadsheetNewSeasonPath(
    execute=@n2.then_4
    XpSpreadsheetNewSeasonPath:String=Append!(A=Get DataTablesPath(), B=Append!(A=AthenaXPLevels_S, B=@n11, C:String=.xlsb) -> (ReturnValue)) -> (ReturnValue)
   ) -> (Output_Get:String)
   n12 = Get XpSpreadsheetPrevSeasonPath()
   Set AllStepsSuccessful(
    execute=@n2.then_5
    AllStepsSuccessful=AND Boolean(A=Get AllStepsSuccessful() -> (AllStepsSuccessful:bool), B=FileExists!(InPath:String=@n12) -> (ReturnValue:bool))
   )
   Set XpTableIsReady(execute=@n1.then_2, XpTableIsReady:bool=Get AllStepsSuccessful()) -> (Output_Get:bool)
   Return Node(execute=@n1.then_3)
  ]}
 SetSeason {nodes [
   n1 = SetSeason!() -> (then, SeasonToBuild:int, SeasonItemDefinition:Athena Season Item Definition Object Reference) {is=title K2Node_FunctionEntry}
   > Set SeasonNumber(SeasonNumber:int=@n1) -> (Output_Get:int=0)
   > Set SeasonItemDefintion(SeasonItemDefintion:Athena Season Item Definition Object Reference=@n1.SeasonItemDefinition) -> (Output_Get:Athena Season Item Definition Object Reference)
   > ReadySeasonUmbrella() {is=title K2Node_CallFunction}
   > ReadyXPDataTable() {is=title K2Node_CallFunction}
  ]}
]
variables [
 Path To Xp Data {type=string, default=""}
 PrevSeasonUmbrella {type=AssetData, default {}}
 SeasonItemDefintion {type=AthenaSeasonItemDefinition, default=null}
 SeasonNumber {type=integer}
 UmbrellaReadyToCopy {type=boolean}
 XpDataTablePrevSeason {type=AssetData, default {}}
 XpDataTablePrevSeasonExtended {type=AssetData, default {}}
 XpSpreadsheetNewSeasonPath {type=string, default=""}
 XpSpreadsheetPrevSeasonPath {type=string, default=""}
 XpTableIsReady {type=boolean}
]
widgets {
 info {rootWidgetClass=&VerticalBox, widgetCount=14}
 widgets [
  VerticalBox_0 {widgetClassPath=&VerticalBox, widget=VerticalBox_0}
  HorizontalBox_55 {widgetClassPath=&HorizontalBox, parent=VerticalBox_0, widget=HorizontalBox_55, slot {class=VerticalBoxSlot}}
  PrevUmbrellaText {
   widgetClassPath=&CommonTextBlock
   bIsVariable=true
   parent=HorizontalBox_55
   widget=PrevUmbrellaText
   properties {
    style=/Game/Blueprints/Blutilities/ToolStyles/TextStyle-EditorTool-Body.TextStyle-EditorTool-Body_C
    text=No Season Umbrella Yet
    font {
     bForceMonospaced=false
     bMaterialIsStencil=false
     fontMaterial=null
     fontObject=&BurbankSmall
     letterSpacing=0
     monospacedWidth=1
     size=10
     skewAmount=0
     typefaceFontName=Medium
     outlineSettings {
      bApplyOutlineToDropShadows=false
      bMiteredCorners=false
      bSeparateFillAlpha=false
      outlineMaterial=null
      outlineSize=0
      outlineColor=<rgba 0 0 0 1>
     }
    }
    shadowOffset=<0 0>
   }
   slot {class=HorizontalBoxSlot, properties {verticalAlignment=VAlign_Center, padding=<ltrb 3 3 3 3>}}
  }
  TA_QuestEditor_Button {
   widgetClassPath=/Game/NeverCook/EditorUtilities/QuestTools/SupportWidget/TA_QuestEditor_Button
   bIsVariable=true
   parent=HorizontalBox_55
   widget=TA_QuestEditor_Button
   properties {label=Create Season Umbrella}
   slot {class=HorizontalBoxSlot}
  }
  UmbrellaOutput {
   widgetClassPath=&CommonTextBlock
   bIsVariable=true
   parent=HorizontalBox_55
   widget=UmbrellaOutput
   properties {
    style=&TextStyle-EditorTool-Body_C
    text=Umbrella Output
    font {
     bForceMonospaced=false
     bMaterialIsStencil=false
     fontMaterial=null
     fontObject=&BurbankSmall
     letterSpacing=0
     monospacedWidth=1
     size=10
     skewAmount=0
     typefaceFontName=Medium
     outlineSettings {
      bApplyOutlineToDropShadows=false
      bMiteredCorners=false
      bSeparateFillAlpha=false
      outlineMaterial=null
      outlineSize=0
      outlineColor=<rgba 0 0 0 1>
     }
    }
    shadowOffset=<0 0>
   }
   slot {class=HorizontalBoxSlot, properties {verticalAlignment=VAlign_Center, padding=<ltrb 3 3 3 3>}}
  }
  HorizontalBox {widgetClassPath=&HorizontalBox, parent=VerticalBox_0, widget=HorizontalBox, slot {class=VerticalBoxSlot}}
  SeasonXPText {
   widgetClassPath=&CommonTextBlock
   bIsVariable=true
   parent=HorizontalBox
   widget=SeasonXPText
   properties {
    style=&TextStyle-EditorTool-Body_C
    text=No Season XP Yet
    font {
     bForceMonospaced=false
     bMaterialIsStencil=false
     fontMaterial=null
     fontObject=&BurbankSmall
     letterSpacing=0
     monospacedWidth=1
     size=10
     skewAmount=0
     typefaceFontName=Medium
     outlineSettings {
      bApplyOutlineToDropShadows=false
      bMiteredCorners=false
      bSeparateFillAlpha=false
      outlineMaterial=null
      outlineSize=0
      outlineColor=<rgba 0 0 0 1>
     }
    }
    shadowOffset=<0 0>
   }
   slot {class=HorizontalBoxSlot, properties {verticalAlignment=VAlign_Center, padding=<ltrb 3 3 3 3>}}
  }
  MakeXPSpreadsheetButton {
   widgetClassPath=&TA_QuestEditor_Button
   bIsVariable=true
   parent=HorizontalBox
   widget=MakeXPSpreadsheetButton
   properties {label=Create XP Tables}
   slot {class=HorizontalBoxSlot}
  }
  XPSpreadsheetOutput {
   widgetClassPath=&CommonTextBlock
   bIsVariable=true
   parent=HorizontalBox
   widget=XPSpreadsheetOutput
   properties {
    style=&TextStyle-EditorTool-Body_C
    text=XP Output
    font {
     bForceMonospaced=false
     bMaterialIsStencil=false
     fontMaterial=null
     fontObject=&BurbankSmall
     letterSpacing=0
     monospacedWidth=1
     size=10
     skewAmount=0
     typefaceFontName=Medium
     outlineSettings {
      bApplyOutlineToDropShadows=false
      bMiteredCorners=false
      bSeparateFillAlpha=false
      outlineMaterial=null
      outlineSize=0
      outlineColor=<rgba 0 0 0 1>
     }
    }
    shadowOffset=<0 0>
   }
   slot {class=HorizontalBoxSlot, properties {verticalAlignment=VAlign_Center, padding=<ltrb 3 3 3 3>}}
  }
  HorizontalBox_1 {widgetClassPath=&HorizontalBox, parent=VerticalBox_0, widget=HorizontalBox_1, slot {class=VerticalBoxSlot}}
  DynamicRestText {
   widgetClassPath=&CommonTextBlock
   bIsVariable=true
   parent=HorizontalBox_1
   widget=DynamicRestText
   properties {
    style=&TextStyle-EditorTool-Body_C
    text=No Season Rest Tables Yet
    font {
     bForceMonospaced=false
     bMaterialIsStencil=false
     fontMaterial=null
     fontObject=&BurbankSmall
     letterSpacing=0
     monospacedWidth=1
     size=10
     skewAmount=0
     typefaceFontName=Medium
     outlineSettings {
      bApplyOutlineToDropShadows=false
      bMiteredCorners=false
      bSeparateFillAlpha=false
      outlineMaterial=null
      outlineSize=0
      outlineColor=<rgba 0 0 0 1>
     }
    }
    shadowOffset=<0 0>
   }
   slot {class=HorizontalBoxSlot, properties {verticalAlignment=VAlign_Center, padding=<ltrb 3 3 3 3>}}
  }
  MakeDynamicRestButton {
   widgetClassPath=&TA_QuestEditor_Button
   bIsVariable=true
   parent=HorizontalBox_1
   widget=MakeDynamicRestButton
   properties {label=Create Rest Tables}
   slot {class=HorizontalBoxSlot}
  }
  DynamicRestOutput {
   widgetClassPath=&CommonTextBlock
   bIsVariable=true
   parent=HorizontalBox_1
   widget=DynamicRestOutput
   properties {
    style=&TextStyle-EditorTool-Body_C
    text=Rest Output
    font {
     bForceMonospaced=false
     bMaterialIsStencil=false
     fontMaterial=null
     fontObject=&BurbankSmall
     letterSpacing=0
     monospacedWidth=1
     size=10
     skewAmount=0
     typefaceFontName=Medium
     outlineSettings {
      bApplyOutlineToDropShadows=false
      bMiteredCorners=false
      bSeparateFillAlpha=false
      outlineMaterial=null
      outlineSize=0
      outlineColor=<rgba 0 0 0 1>
     }
    }
    shadowOffset=<0 0>
   }
   slot {class=HorizontalBoxSlot, properties {verticalAlignment=VAlign_Center, padding=<ltrb 3 3 3 3>}}
  }
  DeleteAfterTesting {
   widgetClassPath=&DetailsView
   parent=VerticalBox_0
   widget=DeleteAfterTesting
   properties {propertiesToShow [SeasonNumber, SeasonItemDefintion]}
   slot {class=VerticalBoxSlot}
  }
 ]
}
