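# Editor utility widget that creates a new Battle Pass season plugin from an existing season, requiring the editor version set to Future.
# It makes a scratch copy of the season item, trims some lists, and shows details views for manual pruning.
# Advanced Copy moves it into BattlePassS<N>/SeasonData.
# A post-copy step deletes scratch assets, runs an embedded hidden Quest Seasoner on milestones, updates catalog offer IDs and starts content validation.
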EditorUtilityWidgetBlueprint /Game/NeverCook/EditorUtilities/QuestTools/BattlePassPluginCopier/EUW_BattlePassPluginCopier
graphs [
 AdvancedCopyToDestination {nodes [
   AdvancedCopyToDestination() {is=title K2Node_FunctionEntry}
   n1 = Break Asset Data(AssetData=CreateAssetData(InAsset=Get ScratchSeasonItem() -> (ScratchSeasonItem:Athena Season Item Definition Object Reference), bAllowBlueprintClass=true))
   n2 = Append(A=GetDestinationPluginPath!() -> (AssetPath:String, AbsolutePath:String).AbsolutePath, B=SeasonData/)
   > PrintString(InString=Append!(A="Advanced Copying ", B="To String (Name)"(InName=@n1), C:String=" to ", D:String=@n2) -> (ReturnValue))
   > AdvancedCopyPackagesToTargetPath!(execute, PackageNamesToCopy:Array of Names=Make Array("[0]":Name=@n1) -> (Array:Array of Names), TargetPath:String=@n2) -> (then)
  ]}
 Check if New Season Plugin already exists {nodes [
   Check if New Season Plugin already exists() {is=title K2Node_FunctionEntry}
   > n1 = GetAssetsByPath!(
    execute
    self:Asset Registry Interface=GetAssetRegistry()
    PackagePath:Name=String To Name(InString=Get destinationPluginPath() -> (destinationPluginPath:String))
    bRecursive:bool=false
    bIncludeOnlyOnDiskAssets:bool=false
   ) -> (then, OutAssetData:Array of Asset Data Structures, ReturnValue:bool)
   n2 = Length(TargetArray:Array of Asset Data Structures=@n1)
   n3 = Get NewSeasonNumber() -> (NewSeasonNumber:int=0)
   > n4 = Branch(Condition=AND Boolean(A="Equal (Integer)"(A=@n2), B="integer > integer"(A=@n3, B=14)))
   > Return Node!(execute, Season Exists:bool=false)
   Return Node(execute=@n4.else, Season Exists=true)
   Reroute Node()
   n5 = "To String (Integer)"(InInt=Get NewSeasonNumber() -> (NewSeasonNumber=0))
   > Show Message Dialog(
    Title=MakeLiteralText(Value="LOCTEXT(\"Season number invalid!\")")
    Message="To Text (String)"(InString=Append!(A=Cannot set up a plugin for BattlePassS, B=@n5, C=" - that season already exists!") -> (ReturnValue))
    MessageType=Ok
    DefaultValue=No
    MessageCategory=Error
   )
   > Return Node(Season Exists=false)
  ]}
 CheckIfEditorIsSetToFuture {nodes [
   CheckIfEditorIsSetToFuture() {is=title K2Node_FunctionEntry}
   > n1 = EditorOnly_GetCurrentFortReleaseVersion!(execute) -> (then, ReturnValue:Fort Release Version Structure)
   > n2 = EditorOnly_GetVersionSeasonNumber!(execute, Version:"Fort Release Version Structure (by ref)"=@n1) -> (then, ReturnValue:int=0)
   > Return Node!(execute, Is Future:bool="Equal (Integer)"(A=@n2))
  ]}
 CleanChallengeSchedulesAlwaysShown {nodes [
   CleanChallengeSchedulesAlwaysShown() {is=title K2Node_FunctionEntry}
   > SetEditorProperty(
    Object=Get ScratchSeasonItem()
    PropertyName=ChallengeSchedulesAlwaysShown
    PropertyValue:Array of Fort Challenge Bundle Schedule Definition Object References=Make Array!() -> (Array:Array of Fort Challenge Bundle Schedule Definition Object References)
   )
   > PrintString(InString=Cleared the ChallengeSchedulesAlwaysShown array)
  ]}
 CleanSeasonCurrencyMcpDataList {nodes [
   CleanSeasonCurrencyMcpDataList!() -> (then, IndexOfItemsToKeep:int) {is=title K2Node_FunctionEntry}
   > n1 = GetEditorProperty!(execute:Exec, Object:Object Reference=Get ScratchSeasonItem(), PropertyName:Name=SeasonCurrencyMcpDataList) -> (then:Exec, PropertyValue:Array of Season Currency Mcp Data Structures, ReturnValue:bool)
   > PrintString(InString=Cleaning up SeasonCurrencyMcpData list)
   n2 = Get IndexOfItemsToKeep() -> (IndexOfItemsToKeep:int=0)
   > SetEditorProperty(
    Object=Get ScratchSeasonItem()
    PropertyName=SeasonCurrencyMcpDataList
    PropertyValue:Array of Season Currency Mcp Data Structures=Make Array!("[0]":Season Currency Mcp Data Structure="Get (a ref)"(Array:Array of Season Currency Mcp Data Structures=@n1, Dimension 1=@n2) -> (Output:"Season Currency Mcp Data Structure (by ref)")) -> (Array:Array of Season Currency Mcp Data Structures)
   )
  ]}
 CleanupAllScratchAssets {nodes [
   CleanupAllScratchAssets() {is=title K2Node_FunctionEntry}
   n1 = Get ScratchSeasonItem()
   > CleanupScratchAssets!(execute, SeasonItem:Athena Season Item Definition Object Reference=@n1) -> (then)
   > DeleteLoadedAsset!(execute, AssetToDelete:Object Reference=@n1) -> (then, ReturnValue:bool)
  ]}
 DoPostCopyStuff {nodes [
   DoPostCopyStuff() {is=title K2Node_FunctionEntry}
   > FindDestinationSeasonObject()
   > CleanupAllScratchAssets() {is=title K2Node_CallFunction}
   > n1 = NewSeasonPostCopy!(
    execute
    SeasonItem:Athena Season Item Definition Object Reference=Get DestinationSeasonItem() -> (DestinationSeasonItem:Athena Season Item Definition Object Reference)
   ) -> (then, ReturnValue:bool)
   > Branch(Condition=@n1)
   n2 = Get NewSeasonNumber() -> (NewSeasonNumber=0)
   > ProgrammaticallyFillSeasoner!(
    execute
    self:EUW Quest Seasoner Object Reference=Get EUW_MilestoneSeasoner_49() -> (EUW_MilestoneSeasoner_49:EUW Quest Seasoner Object Reference)
    MilestoneFolder:String=Append!(A=Get destinationPluginPath(), B="QuestItems\\Milestone\\") -> (ReturnValue)
    OldSeason:int=int - int(A=@n2)
    NewSeason:int=@n2
   ) -> (then)
   > ProgrammaticallyRunSeasoner(self:EUW Quest Seasoner Object Reference=Get EUW_MilestoneSeasoner_49())
   > UpdateOfferIdsFromCatalog!(
    execute
    SeasonDefToUpdate:Athena Season Item Definition Object Reference=Get DestinationSeasonItem()
    SourceCatalogSoftPtr:Catalog Definition Soft Object Reference=/Game/Catalog/StoreCatalog
   ) -> (then, ReturnValue:bool)
   > SaveDirtyPackagesWithDialog!(execute, bSaveMapPackages:bool=false, bSaveContentPackages:bool=true) -> (then, ReturnValue:bool)
   > StartValidateCheckedOutContent!(
    execute
    PackagesToSkip:Array of Asset Data Structures=Make Array("[0]":Asset Data Structure=CreateAssetData(InAsset=&self, bAllowBlueprintClass=false)) -> (Array:Array of Asset Data Structures)
   ) -> (then)
   > SetVisibility(self=Get FinalInstructionsTextBox() -> (FinalInstructionsTextBox:Common Text Object Reference), InVisibility=SelfHitTestInvisible)
   > SetSeason!(
    execute
    self:EUW Battle Pass Asset Creation Widget Object Reference=Get ExtraPostCopyActions() -> (ExtraPostCopyActions:EUW Battle Pass Asset Creation Widget Object Reference)
    SeasonToBuild:int=Get NewSeasonNumber() -> (NewSeasonNumber=0)
    SeasonItemDefinition:Athena Season Item Definition Object Reference=Get DestinationSeasonItem()
   ) -> (then)
  ]}
 EventGraph {nodes [
   "On Clicked (SaveHumanBrainWorkButton)"!() -> (OutputDelegate, then, widget:User Widget Object Reference, event:Pointer Event Structure) {is=title K2Node_ComponentBoundEvent}
   > SaveScratchAssets!(
    execute
    SeasonItem:Athena Season Item Definition Object Reference=Get ScratchSeasonItem()
    bIncludeSeasonItem:bool=true
    bWaitForCompletion:bool=true
   ) -> (then)
   > AdvancedCopyToDestination() {is=title K2Node_CallFunction}
   "On Clicked (PrepareNewSeasonButton)"!() -> (OutputDelegate, then, widget:User Widget Object Reference, event:Pointer Event Structure) {is=title K2Node_ComponentBoundEvent}
   > Prepare New Season()
   > Refresh Season Picker Controls()
   "On Clicked (AdvancedCopyDonebutton)"!() -> (OutputDelegate, then, widget:User Widget Object Reference, event:Pointer Event Structure) {is=title K2Node_ComponentBoundEvent}
   > DoPostCopyStuff() {is=title K2Node_CallFunction}
   "On Property Changed (SeasionItemToCopySelector)"!() -> (OutputDelegate, then, PropertyName:Name=None) {is=title K2Node_ComponentBoundEvent}
   > SetDestinationPluginPath()
   n1 = Get SeasonItemToCopy() -> (SeasonItemToCopy:Athena Season Item Definition Object Reference)
   n2 = Get SeasonNumber!(self:Fort Season Item Definition Object Reference=@n1) -> (SeasonNumber:int=0)
   > Set NewSeasonNumber(NewSeasonNumber:int=Select(Option 0:int=0, Option 1:int=int + int(A=@n2), Index=IsValid(Object=@n1)) -> (ReturnValue:int)) -> (Output_Get:int=0)
   > Refresh Season Picker Controls()
   "On Property Changed (NextSeasonNumberPropertyView)"!() -> (OutputDelegate, then, PropertyName:Name=None) {is=title K2Node_ComponentBoundEvent}
   > SetDestinationPluginPath()
   > Refresh Season Picker Controls()
   Event Construct()
   > SetVisibility(
    self=[
     Get FinalInstructionsTextBox()
     Get UserInputWidgetsVerticalBox() -> (UserInputWidgetsVerticalBox:Vertical Box Object Reference)
     Get ExtraPostCopyActions()
    ]
    InVisibility=Collapsed
   )
   > n3 = Branch(Condition=CheckIfEditorIsSetToFuture!() -> (Is Future:bool) {is=title K2Node_CallFunction})
   > SetVisibility(self=Get SeasonPickerControls() -> (SeasonPickerControls:Grid Panel Object Reference), InVisibility=Visible)
   > SetVisibility(self=Get EditorVersionWarning() -> (EditorVersionWarning:Common Text Object Reference), InVisibility=Collapsed)
   > Refresh Season Picker Controls()
   SetVisibility(execute=@n3.else, self=Get SeasonPickerControls(), InVisibility=Collapsed)
   > SetVisibility(self=Get EditorVersionWarning(), InVisibility=Visible)
   Event PreConstruct()
   > SetObject(
    self=[
     Get NewSeasonPickerHack() -> (NewSeasonPickerHack:Fort Quest Tool Property View Object Reference)
     Get NextSeasonNumberPropertyView() -> (NextSeasonNumberPropertyView:Fort Quest Tool Property View Object Reference)
     Get SeasionItemToCopySelector() -> (SeasionItemToCopySelector:Fort Quest Tool Property View Object Reference)
    ]
    NewObject=Self-Reference()
   )
  ]}
 FindDestinationSeasonObject {nodes [
   FindDestinationSeasonObject() {is=title K2Node_FunctionEntry}
   n1 = Append!(A=GetDestinationPluginPath().AbsolutePath, B=SeasonData/, C:String=GetObjectName(Object=Get ScratchSeasonItem())) -> (ReturnValue)
   > n2 = FindAssetData!(execute, AssetPath:String=@n1) -> (then, ReturnValue:Asset Data Structure)
   > n3 = Cast To AthenaSeasonItemDefinition!(execute, Object=GetAsset(InAssetData=@n2)) -> (then, CastFailed, AsAthena Season Item Definition:Athena Season Item Definition Object Reference) {is=title K2Node_DynamicCast}
   > Set DestinationSeasonItem(DestinationSeasonItem:Athena Season Item Definition Object Reference=@n3) -> (Output_Get:Athena Season Item Definition Object Reference)
   PrintString(execute=@n3.CastFailed, InString=Append!(A="Failed to find a proper season item at the destination: ", B=@n1) -> (ReturnValue))
  ]}
 GetDestinationPluginPath {nodes [
   GetDestinationPluginPath() {is=title K2Node_FunctionEntry}
   n1 = Get destinationPluginPath()
   > Return Node!(execute, AssetPath:String=MakeStandardFilename!(InPath:String=@n1) -> (OutPath:String), AbsolutePath:String=@n1)
  ]}
 Is New Season Number Valid {nodes [
   Is New Season Number Valid() {is=title K2Node_FunctionEntry}
   Get NewSeasonNumber() -> (NewSeasonNumber=0)
  ]}
 IsUsingSeasonGrantsToEveryoneAsset {nodes [
   Inputs()
   > n1 = Branch(Condition=Is Using Season Grants to Everyone Data Asset!(self:Fort Season Item Definition Object Reference=Get SeasonItemToCopy()) -> (ReturnValue:bool))
   > Outputs!(True:Exec, False:Exec=@n1.else)
  ]}
 Prepare New Season {nodes [
   Prepare New Season() {is=title K2Node_FunctionEntry}
   > n1 = CopySeasonItemDefinitionAndSubAssets!(
    execute
    SourceSeasonItem:Athena Season Item Definition Object Reference=Get SeasonItemToCopy()
    NewSeasonNumber:int=Get NewSeasonNumber() -> (NewSeasonNumber=0)
   ) -> (then, ReturnValue:Athena Season Item Definition Object Reference)
   > n2 = Set ScratchSeasonItem(ScratchSeasonItem:Athena Season Item Definition Object Reference=@n1) -> (Output_Get:Athena Season Item Definition Object Reference)
   > n3 = Is Valid(InputObject=@n2)
   Show Message Dialog(
    execute=@n3.Is Not Valid
    Title=MakeLiteralText(Value="LOCTEXT(\"ERROR: Failed to create next season\")")
    Message=MakeLiteralText(Value="LOCTEXT(\"ERROR: Next Season Create on the season item failed! Check the log for more details.\")")
    MessageType=Ok
    DefaultValue=No
    MessageCategory=Warning
   )
   n4 = GetEditorProperty(execute=@n3.Is Valid, Object=Get ScratchSeasonItem(), PropertyName=SeasonGrantsToEveryoneDataAsset) -> (PropertyValue:Object Reference)
   > Set ScratchSeasonGrantsToEveryone(ScratchSeasonGrantsToEveryone:Object Reference=@n4) -> (Output_Get:Object Reference)
   > WaitForCompletion(self:Asset Registry Interface=GetAssetRegistry())
   > CleanSeasonCurrencyMcpDataList!(execute, self, IndexOfItemsToKeep:int=0) -> (then) {is=title K2Node_CallFunction}
   > CleanChallengeSchedulesAlwaysShown() {is=title K2Node_CallFunction}
   > SaveScratchAssets(SeasonItem=Get ScratchSeasonItem(), bIncludeSeasonItem=true, bWaitForCompletion=true)
   > SetUpHumanBrainWork()
  ]}
 Refresh Season Picker Controls {nodes [
   n1 = Check if New Season Plugin already exists!(execute, self) -> (then, Season Exists:bool) {is=title K2Node_CallFunction}
   > n2 = Branch(Condition=@n1)
   n3 = Get NewSeasonNumber() -> (NewSeasonNumber=0)
   > setLabel!(
    execute
    self:TA Base Editor Button Widget Object Reference=Get PrepareNewSeasonButton() -> (PrepareNewSeasonButton:TA Quest Editor Button Object Reference)
    label:String=Append!(A="Season ", B="To String (Integer)"(InInt=@n3), C:String=" already exists!") -> (ReturnValue)
   ) -> (then)
   n4 = Get SeasonNumber(self=Get SeasonItemToCopy()) -> (SeasonNumber=0)
   n5 = Get NewSeasonNumber() -> (NewSeasonNumber=0)
   setLabel(
    execute=[@n8.then, @n2.else]
    self=Get PrepareNewSeasonButton()
    label=Append!(A="Prepare to copy Season ", B="To String (Integer)"(InInt=@n4), C=" as Season ", D:String="To String (Integer)"(InInt=@n5)) -> (ReturnValue)
   )
   > SetIsEnabled(self=Get PrepareNewSeasonButton(), bInIsEnabled=true)
   Refresh Season Picker Controls() {is=title K2Node_FunctionEntry}
   > n6 = Is Valid(InputObject=Get ScratchSeasonItem())
   SetIsEnabled(execute=@n6.Is Not Valid, self=Get SeasonPickerControls(), bInIsEnabled=true)
   > SetIsEnabled(self=[Get NextSeasonNumberPropertyView(), Get PrepareNewSeasonButton()], bInIsEnabled=false)
   > n7 = Is Valid(InputObject=Get SeasonItemToCopy())
   setLabel(execute=@n7.Is Not Valid, self=Get PrepareNewSeasonButton(), label=Pick a season to copy...)
   n8 = SetIsEnabled(execute=@n7.Is Valid, self=Get NextSeasonNumberPropertyView(), bInIsEnabled=true)
   SetIsEnabled(execute=@n6.Is Valid, self=Get SeasonPickerControls(), bInIsEnabled=false)
  ]}
 SetDestinationPluginPath {nodes [
   SetDestinationPluginPath() {is=title K2Node_FunctionEntry}
   n1 = Get SeasonItemToCopy()
   n2 = Get Object Path String(Object=Get SeasonItemToCopy())
   n3 = CollapseRelativeDirectories!(InPath:String=Append!(A=Split!(InPath:String=@n2) -> (PathPart:String, FilenamePart:String, ExtensionPart:String), B=../../../) -> (ReturnValue)) -> (OutPath:String, ReturnValue:bool)
   n4 = "To String (Integer)"(InInt=Get NewSeasonNumber() -> (NewSeasonNumber=0))
   > Set destinationPluginPath(destinationPluginPath:String=Select(Option 0:String, Option 1:String=Append!(A=@n3, B=BattlePassS, C:String=@n4, D:String=/) -> (ReturnValue), Index=IsValid(Object=@n1)) -> (ReturnValue:String)) -> (Output_Get:String)
  ]}
 SetUpHumanBrainWork {nodes [
   SetUpHumanBrainWork() {is=title K2Node_FunctionEntry}
   > SetVisibility(self=Get UserInputWidgetsVerticalBox(), InVisibility=SelfHitTestInvisible)
   > SetVisibility(
    self=Get HumanBrainQuestDetailsViewSeasonGrantsAsset() -> (HumanBrainQuestDetailsViewSeasonGrantsAsset:Fort Quest Tool Details View Object Reference)
    InVisibility=Visible
   )
   > Remove Item!(
    execute
    TargetArray:Array of Names=Get PropertiesToShow!(self:Fort Quest Tool Details View Object Reference=Get HumanBrainQuestDetailsViewSeasonAsset() -> (HumanBrainQuestDetailsViewSeasonAsset:Fort Quest Tool Details View Object Reference)) -> (PropertiesToShow:Array of Names)
    Item:"Name (by ref)"=SeasonGrantsToEveryone
   ) -> (then, ReturnValue:bool)
   > SetObject(
    self=Get HumanBrainQuestDetailsViewSeasonGrantsAsset()
    NewObject=Get ScratchSeasonGrantsToEveryone() -> (ScratchSeasonGrantsToEveryone:Object Reference)
   )
   > SetObject(self=Get HumanBrainQuestDetailsViewSeasonAsset(), NewObject=Get ScratchSeasonItem())
  ]}
]
variables [
 destinationPluginPath {type=string, default=""}
 DestinationSeasonItem {type=AthenaSeasonItemDefinition, default=null}
 NewSeasonNumber {type=integer}
 ScratchSeasonGrantsToEveryone {type=Object, default=null}
 ScratchSeasonItem {type=AthenaSeasonItemDefinition, default=null}
 SeasonItemToCopy {type=AthenaSeasonItemDefinition, default=null}
]
widgets {
 info {rootWidgetClass=&ScrollBox, widgetCount=23}
 widgets [
  ScrollBox_0 {
   widgetClassPath=&ScrollBox
   widget=ScrollBox_0
   properties {
    widgetBarStyle {
     thickness=8
     draggedThumbImage {
      bIsDynamicallyLoaded=false
      drawAs=RoundedBox
      imageType=NoImage
      mirroring=NoMirror
      resourceName=None
      resourceObject=null
      tiling=NoTile
      imageSize=<0 0>
      margin=<ltrb 0 0 0 0>
      outlineSettings {
       bUseBrushTransparency=false
       roundingType=FixedRadius
       width=0
       color {colorUseRule=UseColor_Specified, specifiedColor=<rgba 0 0 0 0>}
       cornerRadii=<4 4 4 4>
      }
      tintColor {colorUseRule=UseColor_Specified, specifiedColor=<rgba 0.2159 0.2159 0.2159 1>}
      uVRegion {bIsValid=false, max=<0 0>, min=<0 0>}
     }
     horizontalBackgroundImage {
      bIsDynamicallyLoaded=false
      drawAs=NoDrawType
      imageType=NoImage
      mirroring=NoMirror
      resourceName=None
      resourceObject=null
      tiling=NoTile
      imageSize=<0 0>
      margin=<ltrb 0 0 0 0>
      outlineSettings {
       bUseBrushTransparency=false
       roundingType=HalfHeightRadius
       width=0
       color {colorUseRule=UseColor_Specified, specifiedColor=<rgba 0 0 0 0>}
       cornerRadii=<0 0 0 0>
      }
      tintColor {colorUseRule=UseColor_Specified, specifiedColor=<rgba 1 1 1 1>}
      uVRegion {bIsValid=false, max=<0 0>, min=<0 0>}
     }
     horizontalBottomSlotImage {
      bIsDynamicallyLoaded=false
      drawAs=NoDrawType
      imageType=NoImage
      mirroring=NoMirror
      resourceName=None
      resourceObject=null
      tiling=NoTile
      imageSize=<0 0>
      margin=<ltrb 0 0 0 0>
      outlineSettings {
       bUseBrushTransparency=false
       roundingType=HalfHeightRadius
       width=0
       color {colorUseRule=UseColor_Specified, specifiedColor=<rgba 0 0 0 0>}
       cornerRadii=<0 0 0 0>
      }
      tintColor {colorUseRule=UseColor_Specified, specifiedColor=<rgba 1 1 1 1>}
      uVRegion {bIsValid=false, max=<0 0>, min=<0 0>}
     }
     horizontalTopSlotImage {
      bIsDynamicallyLoaded=false
      drawAs=NoDrawType
      imageType=NoImage
      mirroring=NoMirror
      resourceName=None
      resourceObject=null
      tiling=NoTile
      imageSize=<0 0>
      margin=<ltrb 0 0 0 0>
      outlineSettings {
       bUseBrushTransparency=false
       roundingType=HalfHeightRadius
       width=0
       color {colorUseRule=UseColor_Specified, specifiedColor=<rgba 0 0 0 0>}
       cornerRadii=<0 0 0 0>
      }
      tintColor {colorUseRule=UseColor_Specified, specifiedColor=<rgba 1 1 1 1>}
      uVRegion {bIsValid=false, max=<0 0>, min=<0 0>}
     }
     hoveredThumbImage {
      bIsDynamicallyLoaded=false
      drawAs=RoundedBox
      imageType=NoImage
      mirroring=NoMirror
      resourceName=None
      resourceObject=null
      tiling=NoTile
      imageSize=<0 0>
      margin=<ltrb 0 0 0 0>
      outlineSettings {
       bUseBrushTransparency=false
       roundingType=FixedRadius
       width=0
       color {colorUseRule=UseColor_Specified, specifiedColor=<rgba 0 0 0 0>}
       cornerRadii=<4 4 4 4>
      }
      tintColor {colorUseRule=UseColor_Specified, specifiedColor=<rgba 0.2159 0.2159 0.2159 1>}
      uVRegion {bIsValid=false, max=<0 0>, min=<0 0>}
     }
     normalThumbImage {
      bIsDynamicallyLoaded=false
      drawAs=RoundedBox
      imageType=NoImage
      mirroring=NoMirror
      resourceName=None
      resourceObject=null
      tiling=NoTile
      imageSize=<0 0>
      margin=<ltrb 0 0 0 0>
      outlineSettings {
       bUseBrushTransparency=false
       roundingType=FixedRadius
       width=0
       color {colorUseRule=UseColor_Specified, specifiedColor=<rgba 0 0 0 0>}
       cornerRadii=<4 4 4 4>
      }
      tintColor {colorUseRule=UseColor_Specified, specifiedColor=<rgba 0.0953 0.0953 0.0953 1>}
      uVRegion {bIsValid=false, max=<0 0>, min=<0 0>}
     }
     verticalBackgroundImage {
      bIsDynamicallyLoaded=false
      drawAs=NoDrawType
      imageType=NoImage
      mirroring=NoMirror
      resourceName=None
      resourceObject=null
      tiling=NoTile
      imageSize=<0 0>
      margin=<ltrb 0 0 0 0>
      outlineSettings {
       bUseBrushTransparency=false
       roundingType=HalfHeightRadius
       width=0
       color {colorUseRule=UseColor_Specified, specifiedColor=<rgba 0 0 0 0>}
       cornerRadii=<0 0 0 0>
      }
      tintColor {colorUseRule=UseColor_Specified, specifiedColor=<rgba 1 1 1 1>}
      uVRegion {bIsValid=false, max=<0 0>, min=<0 0>}
     }
     verticalBottomSlotImage {
      bIsDynamicallyLoaded=false
      drawAs=NoDrawType
      imageType=NoImage
      mirroring=NoMirror
      resourceName=None
      resourceObject=null
      tiling=NoTile
      imageSize=<0 0>
      margin=<ltrb 0 0 0 0>
      outlineSettings {
       bUseBrushTransparency=false
       roundingType=HalfHeightRadius
       width=0
       color {colorUseRule=UseColor_Specified, specifiedColor=<rgba 0 0 0 0>}
       cornerRadii=<0 0 0 0>
      }
      tintColor {colorUseRule=UseColor_Specified, specifiedColor=<rgba 1 1 1 1>}
      uVRegion {bIsValid=false, max=<0 0>, min=<0 0>}
     }
     verticalTopSlotImage {
      bIsDynamicallyLoaded=false
      drawAs=NoDrawType
      imageType=NoImage
      mirroring=NoMirror
      resourceName=None
      resourceObject=null
      tiling=NoTile
      imageSize=<0 0>
      margin=<ltrb 0 0 0 0>
      outlineSettings {
       bUseBrushTransparency=false
       roundingType=HalfHeightRadius
       width=0
       color {colorUseRule=UseColor_Specified, specifiedColor=<rgba 0 0 0 0>}
       cornerRadii=<0 0 0 0>
      }
      tintColor {colorUseRule=UseColor_Specified, specifiedColor=<rgba 1 1 1 1>}
      uVRegion {bIsValid=false, max=<0 0>, min=<0 0>}
     }
    }
    widgetStyle {
     barThickness=8
     bottomShadowBrush {
      bIsDynamicallyLoaded=false
      drawAs=Box
      imageType=FullColor
      mirroring=NoMirror
      resourceName=../../../Engine/Content/Slate/Common/ScrollBoxShadowBottom.png
      resourceObject=null
      tiling=NoTile
      imageSize=<16 8>
      margin=<ltrb 0.5 0 0.5 1>
      outlineSettings {
       bUseBrushTransparency=false
       roundingType=HalfHeightRadius
       width=0
       color {colorUseRule=UseColor_Specified, specifiedColor=<rgba 0 0 0 0>}
       cornerRadii=<0 0 0 0>
      }
      tintColor {colorUseRule=UseColor_Specified, specifiedColor=<rgba 1 1 1 1>}
      uVRegion {bIsValid=false, max=<0 0>, min=<0 0>}
     }
     horizontalScrolledContentPadding=<ltrb 0 0 1 0>
     leftShadowBrush {
      bIsDynamicallyLoaded=false
      drawAs=Box
      imageType=FullColor
      mirroring=NoMirror
      resourceName=../../../Engine/Content/Slate/Common/ScrollBoxShadowLeft.png
      resourceObject=null
      tiling=NoTile
      imageSize=<8 16>
      margin=<ltrb 1 0.5 0 0.5>
      outlineSettings {
       bUseBrushTransparency=false
       roundingType=HalfHeightRadius
       width=0
       color {colorUseRule=UseColor_Specified, specifiedColor=<rgba 0 0 0 0>}
       cornerRadii=<0 0 0 0>
      }
      tintColor {colorUseRule=UseColor_Specified, specifiedColor=<rgba 1 1 1 1>}
      uVRegion {bIsValid=false, max=<0 0>, min=<0 0>}
     }
     rightShadowBrush {
      bIsDynamicallyLoaded=false
      drawAs=Box
      imageType=FullColor
      mirroring=NoMirror
      resourceName=../../../Engine/Content/Slate/Common/ScrollBoxShadowRight.png
      resourceObject=null
      tiling=NoTile
      imageSize=<8 16>
      margin=<ltrb 0 0.5 1 0.5>
      outlineSettings {
       bUseBrushTransparency=false
       roundingType=HalfHeightRadius
       width=0
       color {colorUseRule=UseColor_Specified, specifiedColor=<rgba 0 0 0 0>}
       cornerRadii=<0 0 0 0>
      }
      tintColor {colorUseRule=UseColor_Specified, specifiedColor=<rgba 1 1 1 1>}
      uVRegion {bIsValid=false, max=<0 0>, min=<0 0>}
     }
     topShadowBrush {
      bIsDynamicallyLoaded=false
      drawAs=Box
      imageType=FullColor
      mirroring=NoMirror
      resourceName=../../../Engine/Content/Slate/Common/ScrollBoxShadowTop.png
      resourceObject=null
      tiling=NoTile
      imageSize=<16 8>
      margin=<ltrb 0.5 1 0.5 0>
      outlineSettings {
       bUseBrushTransparency=false
       roundingType=HalfHeightRadius
       width=0
       color {colorUseRule=UseColor_Specified, specifiedColor=<rgba 0 0 0 0>}
       cornerRadii=<0 0 0 0>
      }
      tintColor {colorUseRule=UseColor_Specified, specifiedColor=<rgba 1 1 1 1>}
      uVRegion {bIsValid=false, max=<0 0>, min=<0 0>}
     }
     verticalScrolledContentPadding=<ltrb 0 0 0 1>
    }
   }
  }
  VerticalBox_0 {widgetClassPath=&VerticalBox, parent=ScrollBox_0, widget=VerticalBox_0, slot {class=ScrollBoxSlot}}
  EditorVersionWarning {
   widgetClassPath=&CommonTextBlock
   bIsVariable=true
   parent=VerticalBox_0
   widget=EditorVersionWarning
   properties {
    justification=Center
    style=/FortUILibrary/Styles/Typography/Foundation/TextStyle-Header-M.TextStyle-Header-M_C
    text="Change your version in the editor version selector to Future before running this tool!"
    font {
     bForceMonospaced=false
     bMaterialIsStencil=false
     fontMaterial=null
     fontObject=&Burbank
     letterSpacing=0
     monospacedWidth=1
     size=32
     skewAmount=0
     typefaceFontName=Burbank-Black
     outlineSettings {
      bApplyOutlineToDropShadows=false
      bMiteredCorners=false
      bSeparateFillAlpha=false
      outlineMaterial=null
      outlineSize=0
      outlineColor=<rgba 0 0 0 1>
     }
    }
    shadowOffset=<0 0>
   }
   slot {class=VerticalBoxSlot, properties {size {sizeRule=Fill, value=1}}}
  }
  SeasonPickerControls {
   widgetClassPath=&GridPanel
   bIsVariable=true
   parent=VerticalBox_0
   widget=SeasonPickerControls
   properties {columnFill [0,0], rowFill [0,0]}
   slot {class=VerticalBoxSlot, properties {padding=<ltrb 5 5 5 5>}}
  }
  CommonTextBlock {
   widgetClassPath=&CommonTextBlock
   parent=SeasonPickerControls
   widget=CommonTextBlock
   properties {
    style=/Game/Blueprints/Blutilities/ToolStyles/TextStyle-EditorTool-Body.TextStyle-EditorTool-Body_C
    text=What season are we copying from?
    font {
     bForceMonospaced=false
     bMaterialIsStencil=false
     fontMaterial=null
     fontObject=&BurbankSmall
     letterSpacing=0
     monospacedWidth=1
     size=10
     skewAmount=0
     typefaceFontName=Medium
     outlineSettings {
      bApplyOutlineToDropShadows=false
      bMiteredCorners=false
      bSeparateFillAlpha=false
      outlineMaterial=null
      outlineSize=0
      outlineColor=<rgba 0 0 0 1>
     }
    }
    shadowOffset=<0 0>
   }
   slot {class=GridSlot, properties {verticalAlignment=VAlign_Center, padding=<ltrb 0 0 10 0>}}
  }
  SeasionItemToCopySelector {
   widgetClassPath=/Script/FortniteEditor.FortQuestToolPropertyView
   bIsVariable=true
   parent=SeasonPickerControls
   widget=SeasionItemToCopySelector
   properties {propertiesToShow [SeasonItemToCopy]}
   slot {class=GridSlot, properties {column=1}}
  }
  CommonText {
   widgetClassPath=&CommonTextBlock
   parent=SeasonPickerControls
   widget=CommonText
   properties {
    style=&TextStyle-EditorTool-Body_C
    text=What season are we copying to?
    font {
     bForceMonospaced=false
     bMaterialIsStencil=false
     fontMaterial=null
     fontObject=&BurbankSmall
     letterSpacing=0
     monospacedWidth=1
     size=10
     skewAmount=0
     typefaceFontName=Medium
     outlineSettings {
      bApplyOutlineToDropShadows=false
      bMiteredCorners=false
      bSeparateFillAlpha=false
      outlineMaterial=null
      outlineSize=0
      outlineColor=<rgba 0 0 0 1>
     }
    }
    shadowOffset=<0 0>
   }
   slot {class=GridSlot, properties {row=1, padding=<ltrb 0 0 10 0>}}
  }
  NextSeasonNumberPropertyView {
   widgetClassPath=&FortQuestToolPropertyView
   bIsVariable=true
   parent=SeasonPickerControls
   widget=NextSeasonNumberPropertyView
   properties {propertiesToShow [NewSeasonNumber]}
   slot {class=GridSlot, properties {column=1, row=1}}
  }
  PrepareNewSeasonButton {
   widgetClassPath=/Game/NeverCook/EditorUtilities/QuestTools/SupportWidget/TA_QuestEditor_Button
   bIsVariable=true
   parent=SeasonPickerControls
   widget=PrepareNewSeasonButton
   properties {label=Prepare New Season}
   slot {class=GridSlot, properties {columnSpan=2, row=2}}
  }
  UserInputWidgetsVerticalBox {
   widgetClassPath=&VerticalBox
   bIsVariable=true
   parent=VerticalBox_0
   widget=UserInputWidgetsVerticalBox
   slot {class=VerticalBoxSlot}
  }
  CommonTextBlock_1 {
   widgetClassPath=&CommonTextBlock
   bIsVariable=true
   parent=UserInputWidgetsVerticalBox
   widget=CommonTextBlock_1
   properties {
    autoWrapText=true
    style=&TextStyle-EditorTool-Body_C
    text="The robots don't know what to do with this, please use your human brain to remove the things that should be removed. Thank you. (Human Brain Hint: Narrative, Mission, Punchcard)\r\n\r\nThe Tutorials Schedules To Grant array is shown for reference but in most cases doesn't need to be modified."
    font {
     bForceMonospaced=false
     bMaterialIsStencil=false
     fontMaterial=null
     fontObject=&BurbankSmall
     letterSpacing=0
     monospacedWidth=1
     size=10
     skewAmount=0
     typefaceFontName=Medium
     outlineSettings {
      bApplyOutlineToDropShadows=false
      bMiteredCorners=false
      bSeparateFillAlpha=false
      outlineMaterial=null
      outlineSize=0
      outlineColor=<rgba 0 0 0 1>
     }
    }
    shadowOffset=<0 0>
   }
   slot {class=VerticalBoxSlot}
  }
  HumanBrainQuestDetailsViewSeasonAsset {
   widgetClassPath=/Script/FortniteEditor.FortQuestToolDetailsView
   bIsVariable=true
   parent=UserInputWidgetsVerticalBox
   widget=HumanBrainQuestDetailsViewSeasonAsset
   properties {propertiesToShow [SeasonGrantsToEveryone, TutorialSchedulesToGrant]}
   slot {class=VerticalBoxSlot}
  }
  HumanBrainQuestDetailsViewSeasonGrantsAsset {
   widgetClassPath=&FortQuestToolDetailsView
   bIsVariable=true
   parent=UserInputWidgetsVerticalBox
   widget=HumanBrainQuestDetailsViewSeasonGrantsAsset
   properties {propertiesToShow [RewardScheduleLevel]}
   slot {class=VerticalBoxSlot}
  }
  SaveHumanBrainWorkButton {
   widgetClassPath=&TA_QuestEditor_Button
   bIsVariable=true
   parent=UserInputWidgetsVerticalBox
   widget=SaveHumanBrainWorkButton
   properties {label=Save Human Brain Work And Proceed}
   slot {class=VerticalBoxSlot}
  }
  FinishUpVerticalBox {
   widgetClassPath=&VerticalBox
   bIsVariable=true
   parent=UserInputWidgetsVerticalBox
   widget=FinishUpVerticalBox
   slot {class=VerticalBoxSlot}
  }
  AdvancedCopyDonebutton {
   widgetClassPath=&TA_QuestEditor_Button
   bIsVariable=true
   parent=FinishUpVerticalBox
   widget=AdvancedCopyDonebutton
   properties {label=Press This When Advanced Copy And Save Are Done}
   slot {class=VerticalBoxSlot}
  }
  EUW_MilestoneSeasoner_49 {
   widgetClassPath=/Game/NeverCook/EditorUtilities/QuestTools/QuestSeasoner/EUW_QuestSeasoner
   bIsVariable=true
   parent=FinishUpVerticalBox
   widget=EUW_MilestoneSeasoner_49
   properties {visibility=Collapsed}
   slot {class=VerticalBoxSlot}
  }
  FinalInstructionsTextBox {
   widgetClassPath=&CommonTextBlock
   bIsVariable=true
   parent=FinishUpVerticalBox
   widget=FinalInstructionsTextBox
   properties {
    autoWrapText=true
    style=&TextStyle-EditorTool-Body_C
    text="Content Check was started, make sure there are no errors. Export MCP templates and run MCP and make sure there are not errors. Then submit all the modified files (and maybe have Matthew Radermacher review them)!"
    font {
     bForceMonospaced=false
     bMaterialIsStencil=false
     fontMaterial=null
     fontObject=&BurbankSmall
     letterSpacing=0
     monospacedWidth=1
     size=10
     skewAmount=0
     typefaceFontName=Medium
     outlineSettings {
      bApplyOutlineToDropShadows=false
      bMiteredCorners=false
      bSeparateFillAlpha=false
      outlineMaterial=null
      outlineSize=0
      outlineColor=<rgba 0 0 0 1>
     }
    }
    shadowOffset=<0 0>
   }
   slot {class=VerticalBoxSlot}
  }
  Border_83 {
   widgetClassPath=&Border
   parent=FinishUpVerticalBox
   widget=Border_83
   properties {brushColor=<rgba 0.2833 0.2833 0.2833 1>, padding=<ltrb 5 5 5 5>}
   slot {class=VerticalBoxSlot}
  }
  FinalInstructionsTextBox_1 {
   widgetClassPath=&CommonTextBlock
   bIsVariable=true
   parent=FinishUpVerticalBox
   widget=FinalInstructionsTextBox_1
   properties {
    autoWrapText=true
    style=&TextStyle-EditorTool-Body_C
    text="After you check in, you can use the following buttons to create assets and hook them up:"
    font {
     bForceMonospaced=false
     bMaterialIsStencil=false
     fontMaterial=null
     fontObject=&BurbankSmall
     letterSpacing=0
     monospacedWidth=1
     size=10
     skewAmount=0
     typefaceFontName=Medium
     outlineSettings {
      bApplyOutlineToDropShadows=false
      bMiteredCorners=false
      bSeparateFillAlpha=false
      outlineMaterial=null
      outlineSize=0
      outlineColor=<rgba 0 0 0 1>
     }
    }
    shadowOffset=<0 0>
   }
   slot {class=VerticalBoxSlot}
  }
  HorizontalBox_303 {widgetClassPath=&HorizontalBox, parent=FinishUpVerticalBox, widget=HorizontalBox_303, slot {class=VerticalBoxSlot}}
  NewSeasonPickerHack {
   widgetClassPath=&FortQuestToolPropertyView
   bIsVariable=true
   parent=HorizontalBox_303
   widget=NewSeasonPickerHack
   properties {propertiesToShow [DestinationSeasonItem]}
   slot {class=HorizontalBoxSlot}
  }
  ExtraPostCopyActions {
   widgetClassPath=/Game/NeverCook/EditorUtilities/QuestTools/BattlePassPluginCopier/SupportWidgets/EUW_BattlePassAssetCreationWidget
   bIsVariable=true
   parent=VerticalBox_0
   widget=ExtraPostCopyActions
   properties {visibility=Visible}
   slot {class=VerticalBoxSlot}
  }
 ]
}
